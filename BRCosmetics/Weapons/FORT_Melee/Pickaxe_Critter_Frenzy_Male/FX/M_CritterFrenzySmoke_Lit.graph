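# Translucent lit smoke material for Niagara sprites in the Critter Frenzy pickaxe effects, with film-grain noise.
# Motion-vector 4x4 flipbooks supply normals, opacity and curve-remapped emissive, shaped by dynamic-parameter noise erosion, depth fade and optional camera fade.
# Static switches select time-of-day sun shading computed in emissive or a sphere-mask fake point light.
# Animated static, scratches and vertical lines, projected on a camera-facing virtual plane with FOV-compensated scale, add to emissive.
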Material /BRCosmetics/Weapons/FORT_Melee/Pickaxe_Critter_Frenzy_Male/FX/M_CritterFrenzySmoke_Lit
graph {
 properties {
  bUsedWithNiagaraSprites=true
  blendMode=BLEND_Translucent
  pixelDepthOffsetMode=PDOM_Legacy
  translucencyLightingMode=TLM_VolumetricPerVertexDirectional
  translucentSelfShadowDensityScale=1
  usedShadingModels=Default Lit
  translucentMultipleScatteringExtinction=<rgba 1 1 1 1>
 }
 nodes [
  Constant() {r=0.8}
  Constant() {r=1.2}
  n1 = VectorParameter() {parameterName=Color, defaultValue=<rgba 0.3958 0.3987 0.4 0>}
  n2 = ParticleColor()
  n3 = LinearInterpolate(A=@n1.RGB, B=@n2.RGB, Alpha=ScalarParameter() {parameterName=UseParticleColor})
  n4 = VectorParameter() {group=SphereMask, parameterName=SphereMask_Color, defaultValue=<rgba 2 0.102 0 0>}
  n5 = DynamicParameter() {parameterIndex=3, paramNames [StrengthPointLight, PointLightRadius, Emissive, NormalMultiplier]}
  n6 = LinearInterpolate(
   B=Multiply(A=@n3, B=Multiply(A=@n4.RGB, B=@n5.StrengthPointLight))
   Alpha=SphereMask(
    A=WorldPosition().XYZ
    B=ActorPositionWS()
    Radius=@n5.PointLightRadius
    Hardness=ScalarParameter() {group=SphereMask, parameterName=Hardness}
   )
  )
  n7 = Multiply(
   A=MaterialFunctionCall(
    Color_Morning=VectorParameter() {parameterName=Color Morning Multiplier, defaultValue=<rgba 0.9896 0.268 0.4929 0>}.RGB
    Color_Day=VectorParameter() {parameterName=Color Day Multiplier, defaultValue=<rgba 1 1 1 0>}.RGB
    Color_Evening=VectorParameter() {parameterName=Color Evening Multiplier, defaultValue=<rgba 0.599 0.2901 0.3358 0>}.RGB
    Color_Night=VectorParameter() {parameterName=Color Night Multiplier, defaultValue=<rgba 0.3498 0.4864 0.6458 0>}.RGB
   ) {materialFunction=&Color_Time_of_Day}.Result
   B=@n3
  )
  n8 = Multiply(A=@n7, B=ScalarParameter() {defaultValue=0.7, parameterName=Shadow Multiplier}) {constB=0.74}
  n9 = DynamicParameter() {paramNames [Animation, Unused, Emissive, NormalMultiplier]}
  n10 = Constant() {r=4}
  n11 = TextureObjectParameter() {parameterName=NormalTexture, samplerType=SAMPLERTYPE_Normal, texture=/Game/Effects/Niagara/MasterMaterials/Smoke/Textures/NEW/RT4x4_3_Tex}
  n12 = TextureObjectParameter() {parameterName=MotionVector, samplerType=SAMPLERTYPE_LinearColor, texture=/Game/Effects/Niagara/MasterMaterials/Smoke/Textures/Vector_OF}
  n13 = ScalarParameter() {defaultValue=0.344, group=Motion Vectors, parameterName=MotionVector}
  n14 = MaterialFunctionCall!(
   "Animation  Phase (0-1)"=@n9.Animation
   Number of Columns X=@n10
   Number of Rows Y=@n10
   Texture=@n11
   UVs
   Motion Vector Texture=@n12
   Phase to Debug
   Motion Vector Intensity=@n13
   Double Apply Motion Vectors
   Custom UV Input
   Custom UV2 Input
   Use Custom UVs
  ) {materialFunction=/Engine/Functions/Engine_MaterialFunctions02/Texturing/FlipBook_MotionVectors}
  n15 = Multiply(
   A=Subtract(A=ComponentMask(None=@n14.Result) {b=true, g=true, r=true}, B=Constant3Vector() {constant=<rgba 0.5 0.5 0 0>})
   B=Constant3Vector() {constant=<rgba 2 2 1 0>}
  )
  n16 = CollectionParameter() {
   collection=&FortniteMaterialParameters
   parameterName=SunlightDirectionalVector
  }
  n17 = DotProduct(A=Transform(None=@n15), B=ComponentMask(None=@n16) {b=true, g=true, r=true})
  n18 = TextureObject() {texture=/Game/Effects/Niagara/MasterMaterials/Smoke/Textures/NEW/RT4x4_5_Tex}
  n19 = CurveAtlasRowParameter(CurveTime=ComponentMask(None=MaterialFunctionCall(
     "Animation  Phase (0-1)"=@n9.Animation
     Number of Columns X=@n10
     Number of Rows Y=@n10
     Texture=@n18
     Motion Vector Texture=@n12
     Motion Vector Intensity=@n13
    ) {materialFunction=&FlipBook_MotionVectors}.Result) {g=true}) {
   atlas=/Game/Effects/Niagara/MasterMaterials/Smoke/Textures/CurveAtlas
   curve=/Game/Effects/Niagara/MasterMaterials/Smoke/Textures/EmissiveCurve
   defaultValue=1
   parameterName=EmissiveCurve
  }
  n20 = MaterialFunctionCall(
   LDR Input=@n19
   HDR Intensity Multiplier=Multiply(
    A=ScalarParameter() {defaultValue=100, parameterName=HDR_Intensity}
    B=DynamicParameter() {paramNames [Animation, Unused, Emissive, NormalMultiplier]}.Emissive
   )
  ) {materialFunction=&DeriveHDRfromLDR}
  n21 = TextureCoordinate()
  n22 = DynamicParameter() {
   parameterIndex=1
   defaultValue=<rgba 0 0 0 1>
   paramNames [Emissive - LerpNoise, Emissive - NoiseOffsetX, Emissive - NoiseErosionPower, Emissive - NoiseScale]
  }
  n23 = TextureSample() {samplerType=SAMPLERTYPE_Normal, texture=/Engine/Functions/Engine_MaterialFunctions02/ExampleContent/Textures/SphereRenderNormalsMap}
  n24 = ScalarParameter() {defaultValue=1, parameterName=SphereNormal}
  n25 = DynamicParameter() {paramNames [Animation, Unused, Emissive, NormalMultiplier]}
  n26 = Normalize(VectorInput=LinearInterpolate(A=Constant3Vector() {constant=<rgba 0 0 1 0>}, B=@n23.RGB, Alpha=Multiply(A=@n24, B=@n25.NormalMultiplier)))
  n27 = Add(
   A=Multiply(A=Add(A=@n21, B=@n22.Emissive - NoiseOffsetX), B=@n22.Emissive - NoiseScale) {desc=Scale}
   B=Multiply(A=ComponentMask(None=@n26) {g=true, r=true}) {constB=-0.2}
  )
  n28 = Multiply(
   A=@n20.RemappedRange
   B=Power(
    Base=TextureSampleParameter2D(UVs=@n27) {parameterName=DetailTexture, texture=/Game/Global/Textures/Default/Tileable_Patterns/T_Blowing_Smoke_Pattern2_Tile_DISP}.G
    Exp=@n22.Emissive - NoiseErosionPower
   )
  )
  n29 = DynamicParameter() {
   parameterIndex=1
   defaultValue=<rgba 0 0 0 1>
   paramNames [Emissive - LerpNoise, Emissive - NoiseOffsetX, Emissive - NoiseErosionPower, Emissive - NoiseScale]
  }
  n30 = Clamp(None=@n29.Emissive - LerpNoise)
  n31 = Add(
   A=Multiply(A=@n1.RGB, B=ScalarParameter() {parameterName=Emissive})
   B=StaticSwitchParameter(True=LinearInterpolate(A=@n20.RemappedRange, B=@n28, Alpha=@n30), False=Constant()) {parameterName=UseEmissive}
  )
  n32 = StaticSwitchParameter(True=Add(A=LinearInterpolate(A=@n8, B=@n7, Alpha=@n17), B=@n31), False=@n31) {parameterName=IsShaderUnlit}
  n33 = NamedRerouteUsage()
  n34 = ComponentMask(None=@n33) {r=true}
  n35 = ComponentMask(None=@n33) {g=true}
  n36 = NamedRerouteUsage()
  n37 = Add(A=Add(A=Add(A=@n34, B=@n35), B=ComponentMask(None=@n36) {b=true}), B=NamedRerouteUsage())
  n38 = ScalarParameter() {defaultValue=20, parameterName=NoiseStrenght}
  MaterialFunctionCall(Emissive=Add(A=StaticSwitchParameter(True=@n6, False=@n32) {parameterName=IsUsingFakePointLightt}, B=Multiply(A=@n37, B=@n38) {constB=10})) {materialFunction=&MF_FX_EmissiveAdjustments}
  n39 = Constant3Vector() {constant=<rgba 1 1 0 0>}
  n40 = MaterialFunctionCall() {materialFunction=&CameraDirectionVector}
  n41 = Multiply(A=Constant() {desc=Needs at least a small amount of this or it will break, r=0.25}, B=Normalize(VectorInput=Multiply(A=@n39, B=@n40.Result)))
  n42 = Multiply(A=@n40.Result, B=Constant() {r=25})
  n43 = Add(A=ParticlePositionWS(), B=Constant3Vector() {constant=<rgba 0 0 40 1>})
  n44 = ScalarParameter() {defaultValue=0.8, group=Bittenhead, parameterName=BH_Scale-LowFOV}
  n45 = ScalarParameter() {defaultValue=1.25, group=Bittenhead, parameterName=BH_Scale-HighFOV}
  n46 = Saturate(None=MaterialFunctionCall(
    Input=MaterialFunctionCall() {materialFunction=/Engine/Functions/Engine_MaterialFunctions02/Utility/FOV}."FOV "
    Input Low=Constant() {r=0.414}
    Input High=Constant() {r=1.091}
    Target Low=Constant()
    Target High=Constant() {r=1}
   ) {materialFunction=&RemapValueRange}.Result)
  NamedRerouteDeclaration(None=MaterialFunctionCall(Plane Normal Axis=Add(A=@n41, B=@n40.Result), Plane Center=Add(A=@n42, B=@n43), UV Scale=LinearInterpolate(A=@n44, B=@n45, Alpha=@n46)) {materialFunction=&VirtualPlaneCoordinates}.UVs) {name=ScreenAlignedUVs, nodeColor=<rgba 0 0.8125 1 1>}
  n47 = ScalarParameter() {defaultValue=0.9, group=Bittenhead, parameterName=BH_Noise-OffsetUV}
  n48 = Time()
  n49 = Floor(None=Multiply(A=ScalarParameter() {defaultValue=16, group=Bittenhead, parameterName=BH_Noise-NumberOffsets}, B=Frac(None=@n48)))
  n50 = ScalarParameter() {defaultValue=0.015, group=Bittenhead, parameterName=BH_Noise-Scale}
  n51 = NamedRerouteUsage()
  n52 = TextureSample(UVs=Add(A=Multiply(A=@n47, B=@n49), B=Multiply(A=@n50, B=@n51))) {
   samplerSource=SSM_Wrap_WorldGroupSettings
   samplerType=SAMPLERTYPE_LinearColor
   texture=/Game/Global/Textures/Default/Noise/SparseGaussianNoise256
  }
  n53 = Multiply(A=OneMinus(None=@n52.R), B=ScalarParameter() {defaultValue=0.03, group=Bittenhead, parameterName=BH_NoiseAmt})
  n54 = VectorParameter() {group=Bittenhead, parameterName=BH_Fx-Color, defaultValue=<rgba 1 1 1 1>}
  n55 = ScalarParameter() {defaultValue=0.5, group=Bittenhead, parameterName=BH_FX-Amount-Other}
  n56 = Constant() {r=1}
  NamedRerouteDeclaration(None=Multiply(
    A=Multiply(A=@n53, B=@n54.RGB)
    B=MaterialFunctionCall(Value_Morning=@n55, Value_Day=@n56, Value_Evening=@n55, Value_Night=@n55) {materialFunction=&Scalar_Time_of_Day}.Result
   )) {name=InteriorStaticNoise, nodeColor=<rgba 0.1016 1 0 1>}
  n57 = Multiply(A=NamedRerouteUsage()) {constB=0.0075}
  n58 = Multiply(A=Time()) {constB=0.5}
  n59 = TextureSample(UVs=AppendVector(A=Multiply(A=@n58) {constB=0.333}, B=Multiply(A=@n58) {constB=0.777})) {
   samplerSource=SSM_Wrap_WorldGroupSettings
   samplerType=SAMPLERTYPE_LinearColor
   texture=&SparseGaussianNoise256
  }
  n60 = AppendVector(A=@n59.R, B=@n59.R)
  n61 = TextureSample(UVs=Frac(None=Add(A=@n57, B=@n60))) {
   samplerSource=SSM_Wrap_WorldGroupSettings
   samplerType=SAMPLERTYPE_Masks
   texture=/Game/Characters/Player/Male/Medium/Bodies/M_MED_Critter_Frenzy/Textures/T_FilmGrain_Noise
  }
  NamedRerouteDeclaration(None=@n61.RGB) {name=ScratchesAndHairs, nodeColor=<rgba 0.3359 1 0 1>}
  n62 = Multiply(A=NamedRerouteUsage()) {constB=0.005}
  n63 = Add(A=Multiply(A=@n59.B) {constB=0.25}, B=Multiply(A=Time()) {constB=0.25})
  n64 = AppendVector(A=Multiply(A=Sine(None=@n63)) {constB=0.025}, B=Multiply(A=Time()) {constB=4})
  n65 = TextureSample(UVs=Frac(None=Subtract(A=@n62, B=@n64))) {
   samplerSource=SSM_Wrap_WorldGroupSettings
   samplerType=SAMPLERTYPE_Masks
   texture=&T_FilmGrain_Noise
  }
  n66 = Time()
  n67 = Abs(None=Sine(None=Multiply(A=@n66) {constB=0.5}))
  NamedRerouteDeclaration(None=Multiply(A=@n65.B, B=LinearInterpolate(Alpha=@n67) {constA=0.1})) {name=VerticalLines, nodeColor=<rgba 1 0.9141 0 1>}
  n68 = LinearInterpolate(A=@n23.RGB, B=@n15, Alpha=ScalarParameter() {defaultValue=1, parameterName=NormalTexture})
  Reroute(None=MaterialFunctionCall(BaseNormal=Normalize(VectorInput=@n68), AdditionalNormal=@n26) {materialFunction=&BlendAngleCorrectedNormals}.Result)
  Reroute(None=@n3)
  n69 = SphereMask(
   A=PixelDepth().R
   B=Constant()
   Radius=ScalarParameter() {defaultValue=400, parameterName=NearCameraFadeRadius}
   Hardness=ScalarParameter() {parameterName=NearCameraFadeHardness}
  )
  n70 = DynamicParameter() {
   parameterIndex=2
   defaultValue=<rgba 0 0 0 1>
   paramNames [Smoke - DissolveLerpNoise, Smoke - DissolveNoiseOffsetX, Smoke - DissolveNoiseErosionPower, Smoke - DissolveNoiseScale]
  }
  n71 = Add(A=TextureCoordinate(), B=@n70.Smoke - DissolveNoiseOffsetX)
  n72 = ComponentMask(None=@n26) {g=true, r=true}
  n73 = TextureSampleParameter2D(UVs=Add(A=Multiply(A=@n71, B=@n70.Smoke - DissolveNoiseScale) {desc=Scale}, B=Multiply(A=@n72) {constB=-0.2})) {parameterName=DissolveDetailTexture, samplerType=SAMPLERTYPE_LinearColor, texture=/Game/Global/Textures/Default/Noise/T_Noise01_Inverse}
  n74 = Clamp(None=Power(Base=@n73.G, Exp=@n70.Smoke - DissolveNoiseErosionPower))
  n75 = TextureObject() {texture=/Game/Effects/Niagara/MasterMaterials/Smoke/Textures/NEW/RT4x4_1_Tex}
  n76 = Subtract(
   A=ComponentMask(None=MaterialFunctionCall(
     "Animation  Phase (0-1)"=@n9.Animation
     Number of Columns X=@n10
     Number of Rows Y=@n10
     Texture=@n75
     Motion Vector Texture=@n12
     Motion Vector Intensity=@n13
    ) {materialFunction=&FlipBook_MotionVectors}.Result) {g=true}
   B=Constant() {r=-1}
  )
  n77 = ScalarParameter() {parameterName=Contrast}
  n78 = ComponentMask(None=MaterialFunctionCall(
    "Animation  Phase (0-1)"=@n9.Animation
    Number of Columns X=@n10
    Number of Rows Y=@n10
    Texture=TextureObject() {texture=/Game/Effects/Niagara/MasterMaterials/Smoke/Textures/NEW/RT4x4_4_Tex}
    Motion Vector Texture=@n12
    Motion Vector Intensity=@n13
   ) {materialFunction=&FlipBook_MotionVectors}.Result) {g=true}
  n79 = DepthFade(FadeDistance=LinearInterpolate(
    A=ScalarParameter() {defaultValue=100, parameterName=DepthFade_Low}
    B=ScalarParameter() {defaultValue=10, parameterName=DepthFade_high}
    Alpha=Multiply(A=@n78, B=@n74)
   ))
  n80 = Multiply(
   A=Multiply(
    A=MaterialFunctionCall(In=@n76, Contrast=@n77) {materialFunction=&CheapContrast}.Result
    B=Multiply(A=@n78, B=@n79)
   )
   B=@n2.A
  )
  n81 = DynamicParameter() {
   parameterIndex=2
   defaultValue=<rgba 0 0 0 1>
   paramNames [Smoke - DissolveLerpNoise, Smoke - DissolveNoiseOffsetX, Smoke - DissolveNoiseErosionPower, Smoke - DissolveNoiseScale]
  }
  n82 = LinearInterpolate(A=@n80, B=Multiply(A=@n80, B=@n74), Alpha=Clamp(None=@n81.Smoke - DissolveLerpNoise))
  StaticSwitchParameter(True=Multiply(A=OneMinus(None=@n69), B=@n82), False=@n82) {parameterName=IsUsingCameraFade}
 ]
}
