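# Niagara system that draws the swing trail for the aged male Critter Frenzy pickaxe.
# One ribbon emitter, FN_AnimTrail, spawns two short-lived anim trails from bone indices on the user-supplied skeletal mesh and updates them against it.
# At spawn, a local ColorFromTOD module tints them using Fortnite's active time of day, with separate morning, afternoon, evening and night colors.
# A curve fades alpha out over each particle's life.
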NiagaraSystem /BRCosmetics/Weapons/FORT_Melee/Pickaxe_Critter_Frenzy_Male/FX/NS_Critter_Frenzy_Pickaxe_M_Aged_Trail
dependencies {usedDataInterfaces [&NiagaraDataInterfaceCurve], usedDynamicInputs [], usedModules [&SystemState]}
emitters {FN_AnimTrail {
  emitterUpdateScript [
   &EmitterState()
   &FN_SpawnAnimTrail(Number of Anim Trails=2)
  ]
  particleSpawnScript [
   &FN_InitializeAnimTrailBoneIndecies()
   &FN_InitializeAnimTrail(Color=<rgba 1 1 1 0.55>, Lifetime=0.125)
   &self:ColorFromTOD_01(
    Afternoon:LinearColor=<rgba 1.44 1.52 1.6 2>
    Evening:LinearColor=<rgba 0.75 0.375 0.375 2>
    Morning:LinearColor=<rgba 0.75 0.1875 0.1875 2>
    Night:LinearColor=<rgba 0.7 1 0.875 2>
    TOD:Vector4f=&MakeVector4_FromLinearColor(LinearColor=$NPC.FortniteNPC.FortniteActiveTimeOfDay)
   ) {module=ColorFromTOD_01}
  ]
  particleUpdateScript [
   &SolveForcesAndVelocity(Clamp Velocity=false, Limit Acceleration=false)
   &FN_UpdateAnimTrail(Locked Particle RibbonWidth=1, SkeletalMesh=$User.SkeletalMesh) {module=FN_UpdateAnimTrail001}
   &ScaleColor(
    Scale Alpha=&FloatFromCurve(FloatCurve=DataInterface() {propertyValues {Curve {keys [{interpMode=RCIM_Cubic, tangentMode=RCTM_User, time=0, value=1}, {interpMode=RCIM_Cubic, tangentMode=RCTM_User, time=1, value=0}]}}})
    ScaleA=true
    ScaleRGB=true
   )
  ]
  rendererClasses [&NiagaraRibbonRendererProperties]
  renderers [{rendererClass=&NiagaraRibbonRendererProperties}]
 }}
summary {
 systemName=NS_Critter_Frenzy_Pickaxe_M_Aged_Trail
 emitters [{emitterName=FN_AnimTrail, rendererClasses [&NiagaraRibbonRendererProperties]}]
 userVariables [
  User.RibbonIDOffset {type=NiagaraInt32}
  User.SkeletalMesh {
   type=NiagaraDataInterfaceSkeletalMesh
   defaultValue {
    struct=NiagaraExt_VariableValue_DataInterface
    value {dataInterface=&self:NiagaraDataInterfaceSkeletalMesh_0}
   }
  }
 ]
}
