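# Physics asset for the Critter Frenzy Male pickaxe, giving its two dangling straps bodies and joints for physical simulation.
# Each strap is a chain of small capsule bodies, five links in the first and four in the second, both constrained to a shared strap master body.
# Joints allow wide swing with very little twist, and the second strap swings more freely than the first.
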PhysicsAsset /BRCosmetics/Weapons/FORT_Melee/Pickaxe_Critter_Frenzy_Male/Meshes/Critter_Frenzy_Male_Physics
bodies {
 dyn_strap1_1_jnt [{shapeType=Capsule, length=2, radius=2, shapeName=dyn_strap1_1_jnt_capsule, center=<3.3107 0 0.0001>, rotation=<rot -90 -2.5305 2.5305>}]
 dyn_strap1_2_jnt [{shapeType=Capsule, length=2, radius=2, shapeName=dyn_strap1_2_jnt_capsule, center=<2.9178 0 0>, rotation=<rot -90 -1.8983 1.8982>}]
 dyn_strap1_3_jnt [{shapeType=Capsule, length=2, radius=2, shapeName=dyn_strap1_3_jnt_capsule, center=<2.9103 0 -0.0001>, rotation=<rot -90 0.0593 -0.0594>}]
 dyn_strap1_4_jnt [{shapeType=Capsule, length=2, radius=2, shapeName=dyn_strap1_4_jnt_capsule, center=<2.9511 0 0>, rotation=<rot -90 -0.0395 0.0395>}]
 dyn_strap1_5_jnt [{shapeType=Capsule, length=2, radius=2, shapeName=dyn_strap1_5_jnt_capsule, center=<2.8889 0 0.0002>, rotation=<rot -90 0.0198 -0.0199>}]
 dyn_strap2_1_jnt [{shapeType=Capsule, length=2, radius=2, shapeName=dyn_strap2_1_jnt_capsule, center=<3.2381 0.0002 0>, rotation=<rot -90 -0.0066 0.0066>}]
 dyn_strap2_2_jnt [{shapeType=Capsule, length=2, radius=2, shapeName=dyn_strap2_2_jnt_capsule, center=<2.675 0 0>, rotation=<rot -90 0.0198 -0.0198>}]
 dyn_strap2_3_jnt [{shapeType=Capsule, length=2, radius=2, shapeName=dyn_strap2_3_jnt_capsule, center=<2.7766 0 0>, rotation=<rot -90 0.0033 -0.0032>}]
 dyn_strap2_4_jnt [{shapeType=Capsule, length=2, radius=2, shapeName=dyn_strap2_4_jnt_capsule, center=<2.7837 0 0>, rotation=<rot -90 0.0197 -0.0197>}]
 dyn_strap_master [{shapeType=Capsule, length=0.505, radius=0.505, shapeName=dyn_strap_master_capsule, rotation=<rot 0 -90 0>}]
}
constraints [
 {bone1Name=dyn_strap2_4_jnt, bone2Name=dyn_strap2_3_jnt, swing1LimitDegrees=90, swing2LimitDegrees=90, twistLimitDegrees=5}
 {bone1Name=dyn_strap1_1_jnt, bone2Name=dyn_strap_master, swing1LimitDegrees=60, swing2LimitDegrees=60, twistLimitDegrees=5}
 {bone1Name=dyn_strap1_2_jnt, bone2Name=dyn_strap1_1_jnt, swing1LimitDegrees=60, swing2LimitDegrees=60, twistLimitDegrees=5}
 {bone1Name=dyn_strap1_3_jnt, bone2Name=dyn_strap1_2_jnt, swing1LimitDegrees=60, swing2LimitDegrees=60, twistLimitDegrees=5}
 {bone1Name=dyn_strap1_4_jnt, bone2Name=dyn_strap1_3_jnt, swing1LimitDegrees=60, swing2LimitDegrees=60, twistLimitDegrees=5}
 {bone1Name=dyn_strap1_5_jnt, bone2Name=dyn_strap1_4_jnt, swing1LimitDegrees=60, swing2LimitDegrees=60, twistLimitDegrees=5}
 {bone1Name=dyn_strap2_1_jnt, bone2Name=dyn_strap_master, swing1LimitDegrees=90, swing2LimitDegrees=90, twistLimitDegrees=5}
 {bone1Name=dyn_strap2_2_jnt, bone2Name=dyn_strap2_1_jnt, swing1LimitDegrees=90, swing2LimitDegrees=90, twistLimitDegrees=5}
 {bone1Name=dyn_strap2_3_jnt, bone2Name=dyn_strap2_2_jnt, swing1LimitDegrees=90, swing2LimitDegrees=90, twistLimitDegrees=5}
]
